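# Additive unlit material for the melee jump attack weapon charge effect, used on Niagara mesh particles and previewed on the jump attack sword mesh.
# An inverted, power-shaped V-coordinate gradient forms a falloff along the mesh UVs.
# The gradient is scaled by an intensity parameter, offset, tinted by particle color and passed through MF_FX_EmissiveAdjustments.
# The same gradient multiplies particle alpha.
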Material /CRD_Melee/FX/JumpAttack/M_Melee_JumpAttak_WeaponCharge
graph {
 properties {
  bUsedWithNiagaraMeshParticles=true
  blendMode=BLEND_Additive
  pixelDepthOffsetMode=PDOM_Legacy
  previewMesh=&/SM_Melee_JumpAttack_Sword
  shadingModel=MSM_Unlit
  usedShadingModels=Unlit
 }
 nodes [
  n1 = ParticleColor()
  n2 = TextureCoordinate()
  n3 = Power(Base=OneMinus(None=ComponentMask(None=@n2) {g=true}))
  n4 = Multiply(A=ScalarParameter() {defaultValue=5, parameterName=Intensity Multiplier}, B=@n3)
  MaterialFunctionCall(Emissive=Multiply(A=@n1.RGB, B=Add(A=@n4))) {materialFunction=&MF_FX_EmissiveAdjustments}
  Multiply(A=@n1.A, B=@n3)
 ]
}
